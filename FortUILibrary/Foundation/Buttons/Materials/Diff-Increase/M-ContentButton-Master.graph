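# Translucent UI master material for a content button, drawing a bordered, gradient-filled background with selected, disabled and shine looks.
# A thresholded UV distance forms the border mask, while a tiled, offset exponential radial gradient blends fill colours and background opacity.
# Selected tints and strengthens the border, and IsDisabled reduces the opacity chain to the border mask.
# A rotated sine band, combined with the border mask, blends in the shine colour.
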Material /FortUILibrary/Foundation/Buttons/Materials/Diff-Increase/M-ContentButton-Master
graph {
 properties {
  bUsedWithStaticMesh=false
  blendMode=BLEND_Translucent
  materialDomain=MD_UI
  pixelDepthOffsetMode=PDOM_Legacy
  usedShadingModels=Default Lit
 }
 nodes [
  n1 = TextureCoordinate()
  n2 = Subtract(A=Abs(None=Add(A=@n1) {constB=-0.5}), B=ScalarParameter() {defaultValue=0.398, parameterName=BorderSize})
  n3 = Add(A=MaterialFunctionCall(Vector 2=Ceil(None=@n2)) {materialFunction=/Engine/Functions/Engine_MaterialFunctions02/Utility/VectorLength}.V2 Length) {constB=0}
  n4 = Clamp(None=OneMinus(None=@n3))
  n5 = AppendVector(
   A=ScalarParameter() {defaultValue=0.75, parameterName=X-Tile, sliderMax=1}
   B=ScalarParameter() {defaultValue=1, parameterName=Y-Tile, sliderMax=1}
  )
  n6 = Multiply(A=TextureCoordinate(), B=@n5)
  n7 = Add(
   A=Multiply(A=@n5) {constB=-0.5}
   B=AppendVector(A=ScalarParameter() {parameterName=X-Offset_1, sliderMax=1}, B=ScalarParameter() {parameterName=Y-Offset_1, sliderMax=1})
  )
  n8 = Clamp(None=MaterialFunctionCall(
    UVs=Add(A=@n6, B=@n7)
    CenterPosition=Constant2Vector()
    Radius=ScalarParameter() {defaultValue=0.5, parameterName=Radius, sliderMax=1}
    Density=ScalarParameter() {defaultValue=1.125, parameterName=Density, sliderMax=4}
   ) {materialFunction=&RadialGradientExponential}.RadialGradientExponential)
  n9 = LinearInterpolate(A=ScalarParameter() {defaultValue=0.6, parameterName=Background-Opacity, sliderMax=1}, B=@n4, Alpha=@n8) {constA=0.75, constB=0.75}
  n10 = ScalarParameter() {parameterName=Selected, sliderMax=1}
  n11 = Clamp(None=@n3)
  LinearInterpolate(A=LinearInterpolate(A=@n9, B=Add(A=@n11, B=@n9), Alpha=@n10), B=@n11, Alpha=ScalarParameter() {parameterName=IsDisabled})
  n12 = LinearInterpolate(
   A=VectorParameter() {parameterName=Border Col1, defaultValue=<rgba 0.0097 0.0176 0.063 1>}.RGB
   B=LinearInterpolate(
    A=VectorParameter() {parameterName=Gradient Col2, defaultValue=<rgba 0.0021 0.0999 0.8148 1>}.RGB
    B=VectorParameter() {parameterName=Gradient Col1, defaultValue=<rgba 0.0887 0.3864 0.9301 1>}.RGB
    Alpha=@n8
   )
   Alpha=LinearInterpolate(B=@n4, Alpha=@n8)
  )
  n13 = VectorParameter() {parameterName=Selection-Col, defaultValue=<rgba 1 0.9301 0 1>}
  n14 = Add(
   A=Rotator(Coordinate=TextureCoordinate(), Time=ScalarParameter() {group=Shine, parameterName=ShineRotation}) {speed=6.2832}
   B=AppendVector(A=ScalarParameter() {group=Shine, parameterName=ShineOffset}, B=Constant())
  )
  n15 = Subtract(A=ComponentMask(None=Sine(None=@n14) {period=2}) {r=true}, B=ScalarParameter() {group=Shine, parameterName=ShineWidth}) {constB=0.5}
  n16 = ScalarParameter() {group=Shine, parameterName=ShineEdge}
  n17 = LinearInterpolate(B=Clamp(None=Multiply(A=@n15, B=@n16)), Alpha=@n4) {constB=0}
  LinearInterpolate(
   A=LinearInterpolate(A=@n12, B=LinearInterpolate(A=@n12, B=@n13.RGB, Alpha=@n11), Alpha=@n10)
   B=VectorParameter() {group=Shine, parameterName=ShineCol, defaultValue=<rgba 0.0887 0.3864 0.9301 1>}.RGB
   Alpha=Multiply(A=OneMinus(None=@n17), B=ScalarParameter() {group=Shine, parameterName=ShineOpacity})
  )
  OneMinus(None=Multiply(A=MaterialFunctionCall() {materialFunction=&VectorToRadialValue}.Linear Distance))
 ]
}
